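# Conversation database for the first step of Winterfest Case 6 in Season 33, where an NPC detective assigns the player to investigate Sgt. Winter and find the Naughty list.
# A single entry tag leads to a linear chain of four speech nodes ending in a Continue node.
# The opening choice requires the NPC to carry the GumshoeDark conversation tag and the matching quest to be active.
# The final choice is tagged with the entry's own tag.
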ConversationDatabase /BattlePassS33_Quests/Conversations/Winterfest/Case_6/Quest_S33_Winterfest_P6_Q01
entry_points [{tag=Conversation.Entry.Season33.WF.Quest.06.part.01, destinations [#1]}]
nodes [
 ConversationEntryPointNode_1 {
  id=#1
  class=ConversationEntryPointNode
  outputs [#2]
  properties {entryTag {tagName=Conversation.Entry.Season33.WF.Quest.06.part.01}}
 }
 Speech_C_1 {
  id=#2
  class=Speech_C
  outputs [#3]
  properties {softTaskIcon=&T_UI_NPC_Intel2, generalConfig {defaultMessage=Alright. Time for our final case together.}}
  sub_nodes [
   ConversationChoiceNode_8 {class=ConversationChoiceNode, properties {defaultChoiceDisplayText=Wintervestigation}}
   FortConversationRequirement_ControllerRequirement_1 {
    class=FortConversationRequirement_ControllerRequirement
    properties {
     nodeName=Require: NPC Has Conversation Tag
     participantId {tagName=Conversation.Participant.NPC}
     requirement {class=FortControllerRequirement_NPCHasConversationEntryTag, properties {queryTag {tagName=Conversation.Entry.GumshoeDark}}}
    }
   }
   FortConversationRequirement_ControllerRequirement_2 {
    class=FortConversationRequirement_ControllerRequirement
    properties {
     nodeName=Require: Is Quest Active
     requirement {
      class=FortControllerRequirement_IsQuestActive
      properties {questDefinition=/BattlePassS33_Quests/Items/QuestItems/Winterfest/Case_6/Quest_S33_Winterfest_P6_Q01}
     }
    }
   }
  ]
 }
 Speech_C_2 {
  id=#3
  class=Speech_C
  outputs [#4]
  properties {
   softTaskIcon=&T_UI_NPC_Intel2
   generalConfig {defaultMessage="Nah. These reports, they're off. And I think I know why."}
  }
  sub_nodes [ConversationChoiceNode_0 {class=ConversationChoiceNode, properties {defaultChoiceDisplayText=Who is it this time? Someone's nice grandma?}}]
 }
 Speech_C_6 {
  id=#4
  class=Speech_C
  outputs [#5]
  properties {
   softTaskIcon=&T_UI_NPC_Intel2
   generalConfig {defaultMessage="I need you to investigate the big man himself, Sgt. Winter."}
  }
  sub_nodes [ConversationChoiceNode_1 {class=ConversationChoiceNode, properties {defaultChoiceDisplayText=...}}]
 }
 Speech_C_7 {
  id=#5
  class=Speech_C
  outputs [#6]
  properties {
   softTaskIcon=&T_UI_NPC_Intel2
   generalConfig {defaultMessage=Find me the Naughty list. You ready to do that... Detective?}
  }
  sub_nodes [ConversationChoiceNode_9 {class=ConversationChoiceNode, properties {defaultChoiceDisplayText=...}}]
 }
 Continue_C_2 {
  id=#6
  class=Continue_C
  properties {nodeName="", softTaskIcon=&T_UI_NPC_Intel2}
  sub_nodes [ConversationChoiceNode_11 {
    class=ConversationChoiceNode
    properties {defaultChoiceDisplayText=Let's do this., choiceTags {gameplayTags [{tagName=Conversation.Entry.Season33.WF.Quest.06.part.01}]}}
   }]
 }
]
