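# NPC conversation for part 2 of Season 33 Winterfest quest 06, apparently still using placeholder dialogue text.
# A single entry tag leads to a main-questline choice gated by a completed-quest requirement.
# That opens a linear chain of five speech nodes advanced by Continue choices, with a separate Continue node branching off the first.
# The last two choices both carry the Story.Confrontation.01 tag, and their messages share the "part 4" label.
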ConversationDatabase /BattlePassS33_Quests/Conversations/Winterfest/Case_6/Quest_S33_Winterfest_P6_Q02
entry_points [{tag=Conversation.Entry.Season33.WF.Quest.06.part.02, destinations [#1]}]
nodes [
 ConversationEntryPointNode_0 {
  id=#1
  class=ConversationEntryPointNode
  outputs [#2]
  properties {entryTag {tagName=Conversation.Entry.Season33.WF.Quest.06.part.02}}
 }
 Speech_C_3 {
  id=#2
  class=Speech_C
  outputs [#3, #4]
  properties {softTaskIcon=&T_UI_NPC_Intel2, generalConfig {defaultMessage=Dialogue part 1}}
  sub_nodes [
   ConversationChoiceNode_5 {
    class=ConversationChoiceNode
    properties {defaultChoiceDisplayText=PART 1 TEXT, choiceTags {gameplayTags [{tagName=Conversation.Choice.Category.MainQuestline}]}}
   }
   FortConversationRequirement_ControllerRequirement_4 {
    class=FortConversationRequirement_ControllerRequirement
    properties {nodeName=Require: Has Completed Quest, requirement {class=FortControllerRequirement_HasCompletedQuest, properties=~}}
   }
  ]
 }
 VkConversationNode_Continue_0 {id=#4, class=VkConversationNode_Continue, outputs []}
 Speech_C_28 {
  id=#3
  class=Speech_C
  outputs [#5]
  properties {softTaskIcon=&T_UI_NPC_Intel2, generalConfig {defaultMessage=Dialogue part 2}}
  sub_nodes [ConversationChoiceNode_6 {class=ConversationChoiceNode, properties {defaultChoiceDisplayText=Continue}}]
 }
 Speech_C_0 {
  id=#5
  class=Speech_C
  outputs [#6]
  properties {softTaskIcon=&T_UI_NPC_Intel2, generalConfig {defaultMessage=Dialogue part 3}}
  sub_nodes [ConversationChoiceNode_7 {class=ConversationChoiceNode, properties {defaultChoiceDisplayText=Continue}}]
 }
 Speech_C_4 {
  id=#6
  class=Speech_C
  outputs [#7]
  properties {softTaskIcon=&T_UI_NPC_Intel2, generalConfig {defaultMessage=Dialogue part 4}}
  sub_nodes [ConversationChoiceNode_3 {
    class=ConversationChoiceNode
    properties {defaultChoiceDisplayText=Continue, choiceTags {gameplayTags [{tagName=Athena.Quests.S33.Conversation.Story.Confrontation.01}]}}
   }]
 }
 Speech_C_5 {
  id=#7
  class=Speech_C
  outputs []
  properties {softTaskIcon=&T_UI_NPC_Intel2, generalConfig {defaultMessage=Dialogue part 4}}
  sub_nodes [ConversationChoiceNode_2 {
    class=ConversationChoiceNode
    properties {defaultChoiceDisplayText=Continue, choiceTags {gameplayTags [{tagName=Athena.Quests.S33.Conversation.Story.Confrontation.01}]}}
   }]
 }
]
