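# Winterfest Case 6 part 4 quest conversation in which the detective player confronts Winter about moving Nice names onto the Naughty list.
# The entry tag opens a speech chain gated on the TacticalSanta NPC tag and this quest being active.
# Each speech node's choice supplies the player's line, ending in a branch to expose Winter or join the Naughty list.
# Choices in both ending branches carry the entry tag as a choice tag.
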ConversationDatabase /BattlePassS33_Quests/Conversations/Winterfest/Case_6/Quest_S33_Winterfest_P6_Q04
entry_points [{tag=Conversation.Entry.Season33.WF.Quest.06.part.04, destinations [#1]}]
nodes [
 ConversationEntryPointNode_2 {
  id=#1
  class=ConversationEntryPointNode
  outputs [#2]
  properties {entryTag {tagName=Conversation.Entry.Season33.WF.Quest.06.part.04}}
 }
 Speech_C_9 {
  id=#2
  class=Speech_C
  outputs [#3]
  properties {
   softTaskIcon=&T_UI_NPC_Intel2
   generalConfig {defaultMessage=If it isn't the rookie detective who couldn't find me ONE name for the naughty list.}
  }
  sub_nodes [
   ConversationChoiceNode_10 {class=ConversationChoiceNode, properties {defaultChoiceDisplayText=Wintervestigation}}
   FortConversationRequirement_ControllerRequirement_1 {
    class=FortConversationRequirement_ControllerRequirement
    properties {
     nodeName=Require: NPC Has Conversation Tag
     participantId {tagName=Conversation.Participant.NPC}
     requirement {class=FortControllerRequirement_NPCHasConversationEntryTag, properties {queryTag {tagName=Conversation.Entry.TacticalSanta}}}
    }
   }
   FortConversationRequirement_ControllerRequirement_2 {
    class=FortConversationRequirement_ControllerRequirement
    properties {
     nodeName=Require: Is Quest Active
     requirement {
      class=FortControllerRequirement_IsQuestActive
      properties {questDefinition=/BattlePassS33_Quests/Items/QuestItems/Winterfest/Case_6/Quest_S33_Winterfest_P6_Q04}
     }
    }
   }
  ]
 }
 Speech_C_10 {
  id=#3
  class=Speech_C
  outputs [#4]
  properties {
   softTaskIcon=&T_UI_NPC_Intel2
   generalConfig {defaultMessage="I woulda gotten away with it, too, if you'd just been a little less meticulous."}
  }
  sub_nodes [ConversationChoiceNode_11 {
    class=ConversationChoiceNode
    properties {defaultChoiceDisplayText="You're cooked, Winter. I know you're putting Nice names on the Naughty list."}
   }]
 }
 Speech_C_11 {
  id=#4
  class=Speech_C
  outputs [#5]
  properties {
   softTaskIcon=&T_UI_NPC_Intel2
   generalConfig {defaultMessage="Look, I'm sure you've done the math. If those names are Nice, the only Naughty one..."}
  }
  sub_nodes [ConversationChoiceNode_13 {class=ConversationChoiceNode, properties {defaultChoiceDisplayText="I'm a detective, sergeant. It's what I do."}}]
 }
 Speech_C_12 {
  id=#5
  class=Speech_C
  outputs [#6]
  properties {
   softTaskIcon=&T_UI_NPC_Intel2
   generalConfig {defaultMessage=It is. I'm not allowed to move my own name. Part of the job.}
  }
  sub_nodes [ConversationChoiceNode_15 {class=ConversationChoiceNode, properties {defaultChoiceDisplayText=...is you. That sounds kinda lonely.}}]
 }
 Speech_C_13 {
  id=#6
  class=Speech_C
  outputs [#7, #8]
  properties {softTaskIcon=&T_UI_NPC_Intel2, generalConfig {defaultMessage=If you could keep this quiet... I'd appreciate it.}}
  sub_nodes [ConversationChoiceNode_2 {class=ConversationChoiceNode, properties {defaultChoiceDisplayText=...}}]
 }
 Speech_C_1 {
  id=#8
  class=Speech_C
  outputs [#9]
  properties {
   softTaskIcon=&T_UI_NPC_Intel2
   generalConfig {defaultMessage="Wow, you'd do that for me? After everything I did? Thanks!"}
  }
  sub_nodes [ConversationChoiceNode_1 {
    class=ConversationChoiceNode
    properties {
     defaultChoiceDisplayText=Put me on the Naughty List too. Nobody should be alone at Winterfest.
     choiceTags {gameplayTags [{tagName=Conversation.Entry.Season33.WF.Quest.06.part.04}]}
    }
   }]
 }
 Speech_C_2 {
  id=#9
  class=Speech_C
  outputs []
  properties {softTaskIcon=&T_UI_NPC_Intel2, generalConfig {defaultMessage="Happy Winterfest, detective."}}
  sub_nodes [ConversationChoiceNode_0 {
    class=ConversationChoiceNode
    properties {defaultChoiceDisplayText=..., choiceTags {gameplayTags [{tagName=Conversation.Entry.Season33.WF.Quest.06.part.04}]}}
   }]
 }
 Speech_C_0 {
  id=#7
  class=Speech_C
  outputs []
  properties {
   softTaskIcon=&T_UI_NPC_Intel2
   generalConfig {defaultMessage="I.. guess that's fair. I'm the only Naughty one, I better take my lumps of coal."}
  }
  sub_nodes [ConversationChoiceNode_14 {
    class=ConversationChoiceNode
    properties {
     defaultChoiceDisplayText=You need to know your actions have consequences. I'm telling everyone.
     choiceTags {gameplayTags [{tagName=Conversation.Entry.Season33.WF.Quest.06.part.04}]}
    }
   }]
 }
]
